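# Sound cue for the looping country music radio in the Apollo ButterBarn biome, played as positional point-of-interest music.
# A single looping wave player feeds the og_country_mix_B_radio_11k_2 wave straight to the output.
# Its own attenuation asset handles falloff with a short maximum distance, and the cue routes to the Fort_Music_POI_PSM sound class.
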SoundCue /Game/Sounds/Apollo/Biomes/ButterBarn/og_country_radio_Cue
root=WavePlayer_0
nodes [{id=WavePlayer_0, properties {bLooping=true, soundWaveAssetPtr=&/og_country_mix_B_radio_11k_2}}]
properties {
 attenuationSettings=&/og_country_radio_Att
 maxDistance=1300
 soundClassObject=/Game/Sounds/FortSoundClassesAndModes/Fort_Music_POI_PSM
 volumeMultiplier=1
}
